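# Volume-domain Niagara sprite material for shield generator FX, rendering each particle as a soft, noisy volumetric puff.
# A sphere mask from particle position and radius, scaled by Extinction, is multiplied by the minimum of two panning dust noise samples remapped between NoiseMin and NoiseMax.
# Low and Medium quality replace the noise with a constant.
# Particle color tints emissive through MF_FX_EmissiveAdjustments, and particle alpha scales albedo and extinction.
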Material /MedCloud/Assets/FX/ShieldGenerator/FX/M_Volumetric_Sprite_ShieldGenerator
graph {
 properties {
  bUsedWithNiagaraSprites=true
  blendMode=BLEND_Additive
  materialDomain=MD_Volume
  pixelDepthOffsetMode=PDOM_Legacy
  usedShadingModels=Default Lit
 }
 nodes [
  n1 = ParticleColor()
  n2 = WorldPosition()
  n3 = ParticlePositionWS()
  n4 = ParticleRadius()
  n5 = ScalarParameter() {defaultValue=0.5, parameterName=Hardness}
  n6 = TextureCoordinate() {uTiling=0.5, vTiling=0.5}
  n7 = Multiply(
   A=Time()
   B=Multiply(
    A=ScalarParameter() {defaultValue=1, parameterName=NoiseSpeed}
    B=DynamicParameter() {defaultValue=<rgba 0.1 3500 1 1>, paramNames [NoiseSpeed,NoiseSize,Param3,Param4]}.NoiseSpeed
   )
  )
  n8 = TextureSampleParameter2D!(UVs=Panner(Coordinate=@n6, Time=@n7) {speedY=-0.1}, Level, Apply View MipBias) {constMipValue=0, mipValueMode=TMVM_MipLevel, parameterName=NoiseTexture, texture=/Game/Global/Textures/Default/PointSplatter/T_Tiling_Dust}
  n9 = ComponentMask(None=VectorParameter() {parameterName=Noise Detail UVS, defaultValue=<rgba 10 5 0 0>}.RGB) {g=true, r=true}
  n10 = TextureSampleParameter2D(UVs=Panner(Coordinate=Multiply(A=@n6, B=@n9), Time=@n7) {speedY=-0.1333}) {constMipValue=0, mipValueMode=TMVM_MipLevel, parameterName=NoiseTexture, texture=&T_Tiling_Dust}
  n11 = Clamp(None=LinearInterpolate(A=ScalarParameter() {parameterName=NoiseMin}, B=ScalarParameter() {defaultValue=1, parameterName=NoiseMax}, Alpha=Min(A=@n8.R, B=@n10.R)))
  n12 = ScalarParameter() {defaultValue=0.5, parameterName=NoNoiseOpacityValue}
  n13 = Multiply(
   A=Multiply(A=SphereMask(A=@n2.XYZ, B=@n3, Radius=@n4, Hardness=@n5), B=ScalarParameter() {defaultValue=3, parameterName=Extinction})
   B=QualitySwitch(Default=@n11, Low=@n12, High=@n11, Medium=@n12)
  )
  MaterialFunctionCall(Emissive=Multiply(A=@n1.RGB, B=@n13)) {materialFunction=&MF_FX_EmissiveAdjustments}
  Multiply(A=VectorParameter() {parameterName=Albedo, defaultValue=<rgba 1 1 1 0>}.RGB, B=@n1.A)
  Multiply(A=@n13, B=@n1.A)
 ]
}
